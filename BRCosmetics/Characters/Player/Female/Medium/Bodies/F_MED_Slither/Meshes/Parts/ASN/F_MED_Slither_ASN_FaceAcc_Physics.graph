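# Physics asset for the face-accessory mesh of the female medium Slither body (ASN part), setting up capsule physics for its dynamic hair bones.
# Five hair roots (01, 02, 03, 05, 06) constrain to the head with zero-degree twist and mostly narrow swing limits.
# Hair 05 continues as a four-bone chain with wider, asymmetric swing.
# Head, neck, upper-spine and left-clavicle capsules form the body side, linked by free angular constraints.
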PhysicsAsset /BRCosmetics/Characters/Player/Female/Medium/Bodies/F_MED_Slither/Meshes/Parts/ASN/F_MED_Slither_ASN_FaceAcc_Physics
bodies {
 clavicle_l [{shapeType=Capsule, length=21.5, radius=3.5, shapeName=clavicle_l_capsule, rotation=<rot 84.1134 180 180>}]
 dyn_hair_01 [{shapeType=Capsule, length=2, radius=2, shapeName=dyn_hair_01_capsule, center=<2 0 0.6>, rotation=<rot 90 0 0>}]
 dyn_hair_02 [{shapeType=Capsule, length=2, radius=2, shapeName=dyn_hair_02_capsule, center=<2.1977 0 0.4>, rotation=<rot 64.1331 0 0>}]
 dyn_hair_03 [{shapeType=Capsule, length=2, radius=2, shapeName=dyn_hair_03_capsule, center=<2 0 1.8>, rotation=<rot 90 0 0>}]
 dyn_hair_05 [{shapeType=Capsule, length=1.5, radius=2.5, shapeName=dyn_hair_05_capsule, center=<2 0 0>, rotation=<rot 90 0 0>}]
 dyn_hair_05_a [{
   shapeType=Capsule
   length=2.5
   radius=2.5
   shapeName=dyn_hair_05_a_capsule
   center=<3.3727 -0.0001 -0.1101>
   rotation=<rot 85.4154 -0.0008 -0.0009>
  }]
 dyn_hair_05_b [{shapeType=Capsule, length=1, radius=3, shapeName=dyn_hair_05_b_capsule, center=<2 0 0>, rotation=<rot 90 0 0>}]
 dyn_hair_05_c [{shapeType=Capsule, length=2.25, radius=1.75, shapeName=dyn_hair_05_c_capsule, center=<2 0 0>, rotation=<rot 90 0 0>}]
 dyn_hair_06 [{shapeType=Capsule, length=2, radius=2, shapeName=dyn_hair_01_capsule, center=<2 0 2>, rotation=<rot 90 0 0>}]
 head [{shapeType=Capsule, length=0, radius=8.3681, shapeName=head_capsule, center=<5.1648 0.504 0.0002>}]
 neck_01 [{shapeType=Capsule, length=2, radius=4.005, shapeName=neck_02_capsule, center=<2.5525 0 0>, rotation=<rot 90 0 0>}]
 neck_02 [{shapeType=Capsule, length=2, radius=4.005, shapeName=neck_02_capsule, center=<2.5525 0 0>, rotation=<rot 90 0 0>}]
 spine_05 [{shapeType=Capsule, length=5.755, radius=10.255, shapeName=spine_05_capsule, center=<1.7636 -1.566 -0.0002>, rotation=<rot 0 -89.9999 0>}]
}
constraints [
 {bone1Name=dyn_hair_05_a, bone2Name=dyn_hair_05, swing1LimitDegrees=30, swing2LimitDegrees=90, twistLimitDegrees=0}
 {bone1Name=dyn_hair_05_b, bone2Name=dyn_hair_05_a, swing1LimitDegrees=30, swing2LimitDegrees=90, twistLimitDegrees=0}
 {bone1Name=dyn_hair_05_c, bone2Name=dyn_hair_05_b, swing1LimitDegrees=30, swing2LimitDegrees=90, twistLimitDegrees=0}
 {bone1Name=dyn_hair_01, bone2Name=head, swing1LimitDegrees=10, swing2LimitDegrees=10, twistLimitDegrees=0}
 {bone1Name=dyn_hair_02, bone2Name=head, swing1LimitDegrees=10, swing2LimitDegrees=10, twistLimitDegrees=0}
 {bone1Name=dyn_hair_03, bone2Name=head, swing1LimitDegrees=10, swing2LimitDegrees=10, twistLimitDegrees=0}
 {bone1Name=dyn_hair_05, bone2Name=head, swing1LimitDegrees=5, swing2LimitDegrees=90, twistLimitDegrees=0}
 {bone1Name=dyn_hair_06, bone2Name=head, swing1LimitDegrees=10, swing2LimitDegrees=5, twistLimitDegrees=0}
 {
  bone1Name=head
  bone2Name=neck_02
  swing1LimitDegrees=45
  swing1Motion=Free
  swing2LimitDegrees=45
  swing2Motion=Free
  twistLimitDegrees=45
  twistMotion=Free
 }
 {
  bone1Name=neck_02
  bone2Name=neck_01
  swing1LimitDegrees=45
  swing1Motion=Free
  swing2LimitDegrees=45
  swing2Motion=Free
  twistLimitDegrees=45
  twistMotion=Free
 }
 {
  bone1Name=neck_01
  bone2Name=spine_05
  swing1LimitDegrees=45
  swing1Motion=Free
  swing2LimitDegrees=45
  swing2Motion=Free
  twistLimitDegrees=45
  twistMotion=Free
 }
 {
  bone1Name=clavicle_l
  bone2Name=spine_05
  swing1LimitDegrees=45
  swing1Motion=Free
  swing2LimitDegrees=45
  swing2Motion=Free
  twistLimitDegrees=45
  twistMotion=Free
 }
]
