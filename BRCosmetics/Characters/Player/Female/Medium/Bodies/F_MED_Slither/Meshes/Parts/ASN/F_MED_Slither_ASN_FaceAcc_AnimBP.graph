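# Anim Blueprint for the F_MED_Slither face accessory, adding secondary motion to the headband and hair on a pose copied from the parent mesh.
# Three head-relative RigidBody nodes run in series: headband physics outside skydiving, a skydive variant, and general physics outside the front end.
# In the front end, hair bones constrain to dedicated _frontend bones, with AnimDynamics and additive offsets layered on.
# The event graph converts state booleans into float alphas driving these blends.
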AnimBlueprint /BRCosmetics/Characters/Player/Female/Medium/Bodies/F_MED_Slither/Meshes/Parts/ASN/F_MED_Slither_ASN_FaceAcc_AnimBP
target_skeleton=&/F_MED_Slither_ASN_FaceAcc_Skeleton
graphs [
 AnimGraph {nodes [
   n1 = RigidBody(
    ComponentPose=Local To Component(LocalPose=Copy Pose From Mesh() {bUseAttachedParent=true})
    Alpha=Get Not Skydiving() -> (Not Skydiving:double)
   ) {
    bClampLinearTranslationLimitToRefPose=true
    bOverrideWorldGravity=true
    overridePhysicsAsset=&/F_MED_Slither_ASN_FaceAcc_PhysicsHeadband
    simulationSpace=BaseBoneSpace
    alphaBoolBlend {blendOption=Linear}
    baseBoneRef {boneName=head}
    componentAppliedLinearAccClamp=<3000 3000 3000>
    componentLinearAccScale=<0.4 0.3 0.3>
    componentLinearVelScale=<0.4 0.3 0.3>
    overrideWorldGravity {z=-980}
   }
   n2 = RigidBody(ComponentPose=@n1, Alpha=Get Is Skydiving() -> (Is Skydiving:double)) {
    bClampLinearTranslationLimitToRefPose=true
    bOverrideWorldGravity=true
    overridePhysicsAsset=&/F_MED_Slither_ASN_FaceAcc_PhysicsHeadbandSkydive
    simulationSpace=BaseBoneSpace
    alphaBoolBlend {blendOption=Linear}
    baseBoneRef {boneName=head}
    componentAppliedLinearAccClamp=<3000 3000 3000>
    componentLinearAccScale=<0.4 0.3 0.3>
    componentLinearVelScale=<0.4 0.3 0.3>
    overrideWorldGravity {z=-980}
   }
   n3 = RigidBody(ComponentPose=@n2, Alpha=Get Is NOT In Front End() -> (Is NOT In Front End:double)) {
    bClampLinearTranslationLimitToRefPose=true
    bOverrideWorldGravity=true
    overridePhysicsAsset=&/F_MED_Slither_ASN_FaceAcc_Physics
    simulationSpace=BaseBoneSpace
    alphaBoolBlend {blendOption=Linear}
    baseBoneRef {boneName=head}
    componentAppliedLinearAccClamp=<3000 3000 3000>
    componentLinearAccScale=<0.4 0.3 0.3>
    componentLinearVelScale=<0.4 0.3 0.3>
    overrideWorldGravity {z=-980}
   }
   n4 = "Transform (Modify) Bone - Bone: dyn_hair_05"(ComponentPose=@n3, Alpha=Get Is Skydiving(), Translation="1,1,1", Rotation="0,0,0", Scale="1,1,1") {
    rotationMode=BMM_Additive
    scaleMode=BMM_Additive
    translationMode=BMM_Additive
    alphaBoolBlend {blendOption=Linear}
    alphaScaleBiasClamp {bInterpResult=true, interpSpeedDecreasing=1, interpSpeedIncreasing=1}
    boneToModify {boneName=dyn_hair_05}
    translation=<1 1 1>
   }
   n5 = Get Is In Front End() -> (Is In Front End:double)
   n6 = "Constraint - dyn_hair_05_a (P)"(ComponentPose=@n4, Alpha=@n5, ConstraintWeights_0=1) {
    alphaBoolBlend {blendOption=Linear}
    boneToModify {boneName=dyn_hair_05_a}
    constraintSetup [{targetBone {boneName=dyn_hair_05_a_frontend}}]
    constraintWeights [1]
   }
   n7 = "Constraint - dyn_hair_05_b (P)"(ComponentPose=@n6, Alpha=@n5, ConstraintWeights_0=1) {
    alphaBoolBlend {blendOption=Linear}
    boneToModify {boneName=dyn_hair_05_b}
    constraintSetup [{targetBone {boneName=dyn_hair_05_b_frontend}}]
    constraintWeights [1]
   }
   n8 = "Constraint - dyn_hair_05_c (P)"(ComponentPose=@n7, Alpha=@n5, ConstraintWeights_0=1) {
    alphaBoolBlend {blendOption=Linear}
    boneToModify {boneName=dyn_hair_05_c}
    constraintSetup [{targetBone {boneName=dyn_hair_05_c_frontend}}]
    constraintWeights [1]
   }
   n9 = Anim Dynamics - Bone: dyn_hair_05_c(ComponentPose=@n8, Alpha=@n5, ExternalForce="0,0,0") {
    angularDampingOverride=1
    angularSpringConstant=1500
    bAngularSpring=true
    bGravityOverrideInSimSpace=true
    bOverrideAngularDamping=true
    bUsePlanarLimit=false
    bUseSphericalLimits=true
    gravityScale=0.35
    numSolverIterationsPreUpdate=8
    simulationSpace=RootRelative
    alphaBoolBlend {blendOption=Linear}
    boundBone {boneName=dyn_hair_05_c}
    physicsBodyDefinitions [{
      collisionType=CustomSphere
      sphereCollisionRadius=2
      boundBone {boneName=dyn_hair_05_c}
      boxExtents=<12 15 15>
      constraintSetup {angularLimitsMax=<0 6 10>, angularLimitsMin=<0 -15 -10>, angularTarget=<0 0 0>}
      localJointOffset=<8 0 0>
     }]
    sphericalLimits [
     {limitRadius=8, drivingBone {boneName=head}, sphereLocalOffset=<0 0 0>}
     {limitRadius=5, drivingBone {boneName=clavicle_l}, sphereLocalOffset=<5 0 0>}
    ]
   }
   n10 = Get bIsInFrontEnd()
   n11 = "Transform (Modify) Bone - Bone: dyn_hair_05_a"!(ComponentPose=@n9, bAlphaBoolEnabled:bool=@n10, Translation="0,-1,0", Rotation="0,-15,0", Scale="1,1,1") -> (Pose) {
    alphaInputType=Bool
    rotationMode=BMM_Additive
    rotationSpace=BCS_BoneSpace
    translationMode=BMM_Additive
    translationSpace=BCS_BoneSpace
    alphaBoolBlend {blendInTime=0.5, blendOption=QuarticInOut}
    boneToModify {boneName=dyn_hair_05_a}
    rotation {yaw=-15}
    translation {y=-1}
   }
   n12 = "Transform (Modify) Bone - Bone: dyn_hair_05_b"!(ComponentPose=@n11, bAlphaBoolEnabled:bool=@n10, Translation="0,-0.5,0", Rotation="0,-5,0", Scale="1,1,1") -> (Pose) {
    alphaInputType=Bool
    rotationMode=BMM_Additive
    rotationSpace=BCS_BoneSpace
    translationMode=BMM_Additive
    translationSpace=BCS_BoneSpace
    alphaBoolBlend {blendInTime=0.5, blendOption=QuarticInOut}
    boneToModify {boneName=dyn_hair_05_b}
    rotation {yaw=-5}
    translation {y=-0.5}
   }
   Output Pose(Result=Component To Local(ComponentPose=@n12)) {layerGroup=None}
  ]}
 EventGraph {nodes [
   Event BlueprintUpdateAnimation()
   > Set Is Sky Diving(Is Sky Diving:double="To Float (Boolean)"(InBool=Get bIsSkyDiving())) -> (Output_Get:double)
   n1 = Get bIsInFrontEnd()
   > Set Is NOT In Front End(Is NOT In Front End:double="To Float (Boolean)"(InBool=NOT Boolean(A=@n1))) -> (Output_Get:double)
   > Set Is In Front End(Is In Front End:double="To Float (Boolean)"(InBool=Get bIsInFrontEnd())) -> (Output_Get:double)
   > Set frontend emote(frontend emote:bool=AND Boolean(A=Get bIsInFrontEnd(), B=Get bIsPlayingEmote())) -> (Output_Get:bool)
   n2 = Get bIsSkyDiving()
   > Set Not Skydiving(Not Skydiving:double="To Float (Boolean)"(InBool=NOT Boolean(A=@n2))) -> (Output_Get:double)
   > n3 = Branch(Condition=Get bIsSkyDiving())
   > Set Is Skydiving(Is Skydiving:double=1) -> (Output_Get:double)
   Set Is Skydiving(execute=@n3.else, Is Skydiving=0)
  ]}
]
variables [
 frontend emote {type=boolean}
 Is In Front End {type=number}
 Is NOT In Front End {type=number}
 Is Sky Diving {type=number}
 Is Skydiving {type=number, default=1}
 Not Skydiving {type=number}
]
